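# Physics asset for the Summer Slurp Monster backpack cosmetic, defining physics bodies and joints for its five dyn_anemone appendage groups.
# Each bone gets a single capsule, and every group's root bone is constrained to the pack body.
# Groups 1 to 3 branch from a hub bone into short sub-chains, while groups 4 and 5 are simple two-bone chains.
# All joints share tight, uniform swing and twist limit angles.
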PhysicsAsset /BRCosmetics/Accessories/FORT_Backpacks/M_MED_Slurp_Monster_Summer/Meshes/M_MED_Slurp_Monster_Summer_Pack_PhysicsAsset
bodies {
 dyn_anemone1_01 [{shapeType=Capsule, length=1.5, radius=1.5}]
 dyn_anemone1_02 [{shapeType=Capsule, length=1.5101, radius=1.3176, center=<1.3539 -1.1367 -0.2891>, rotation=<rot 0 132.7844 -84.8475>}]
 dyn_anemone1_03 [{shapeType=Capsule, length=0.25, radius=0.75, rotation=<rot 90 0 0>}]
 dyn_anemone1_04 [{shapeType=Capsule, length=0.25, radius=0.75, rotation=<rot 90 0 0>}]
 dyn_anemone1_05 [{shapeType=Capsule, length=0.25, radius=0.75, rotation=<rot 90 0 0>}]
 dyn_anemone1_06 [{shapeType=Capsule, length=0.25, radius=0.75, rotation=<rot 90 0 0>}]
 dyn_anemone1_07 [{shapeType=Capsule, length=0.25, radius=0.75, rotation=<rot 90 0 0>}]
 dyn_anemone1_08 [{shapeType=Capsule, length=0.25, radius=0.75, rotation=<rot 90 0 0>}]
 dyn_anemone1_09 [{shapeType=Capsule, length=0.25, radius=0.75, rotation=<rot 90 0 0>}]
 dyn_anemone1_10 [{shapeType=Capsule, length=0.25, radius=0.75, rotation=<rot 90 0 0>}]
 dyn_anemone1_11 [{shapeType=Capsule, length=0.25, radius=0.75, rotation=<rot 90 0 0>}]
 dyn_anemone1_12 [{shapeType=Capsule, length=0.25, radius=0.75, rotation=<rot 90 0 0>}]
 dyn_anemone1_13 [{shapeType=Capsule, length=0.25, radius=0.75, rotation=<rot 90 0 0>}]
 dyn_anemone1_14 [{shapeType=Capsule, length=0.25, radius=0.75, rotation=<rot 90 0 0>}]
 dyn_anemone1_15 [{shapeType=Capsule, length=0.25, radius=0.75, rotation=<rot 90 0 0>}]
 dyn_anemone1_16 [{shapeType=Capsule, length=0.25, radius=0.75, rotation=<rot 90 0 0>}]
 dyn_anemone1_17 [{shapeType=Capsule, length=0.25, radius=0.75, rotation=<rot 90 0 0>}]
 dyn_anemone1_18 [{shapeType=Capsule, length=0.25, radius=0.75, rotation=<rot 90 0 0>}]
 dyn_anemone1_19 [{shapeType=Capsule, length=0.25, radius=0.75, rotation=<rot 90 0 0>}]
 dyn_anemone2_01 [{shapeType=Capsule, length=1.8403, radius=0.6529, center=<0.9876 -0.3239 -0.2083>, rotation=<rot 25.9524 127.9892 -90>}]
 dyn_anemone2_02 [{shapeType=Capsule, length=0.26, radius=0.5, center=<0.7972 0.4527 -0.0471>, rotation=<rot 39.1729 -90 11.693>}]
 dyn_anemone2_03 [{shapeType=Capsule, length=0.26, radius=0.5, rotation=<rot 0 -90 0>}]
 dyn_anemone2_04 [{shapeType=Capsule, length=1.1023, radius=0.75, center=<1.3978 -0.1192 -0.3164>, rotation=<rot 0 90.3151 -83.3013>}]
 dyn_anemone2_05 [{shapeType=Capsule, length=1.6177, radius=0.75, center=<2.1615 0.0458 -0.0285>, rotation=<rot 0 -72.193 -88.4374>}]
 dyn_anemone2_06 [{shapeType=Capsule, length=1.01, radius=0.75, center=<-0.6648 0 0>, rotation=<rot 0 -90 90.0004>}]
 dyn_anemone2_07 [{shapeType=Capsule, length=1.3642, radius=0.5, center=<1.1254 -0.0744 -0.072>, rotation=<rot 0 -175.3236 -67.7654>}]
 dyn_anemone2_08 [{shapeType=Capsule, length=0.26, radius=0.5, rotation=<rot 0 -90 0>}]
 dyn_anemone2_09 [{shapeType=Capsule, length=1.3074, radius=0.5, center=<1.2616 -0.0529 0.1953>, rotation=<rot -18.7388 -90 22.6774>}]
 dyn_anemone2_10 [{shapeType=Capsule, length=0.26, radius=0.5, rotation=<rot 0 -90 0>}]
 dyn_anemone3_01 [{shapeType=Capsule, length=1, radius=1, center=<-1.2657 -0.613 0.9676>, rotation=<rot 31.9696 -90 -33.445>}]
 dyn_anemone3_04 [{shapeType=Capsule, length=1, radius=1, center=<2.6923 0.3835 -0.2162>, rotation=<rot 0 111.1386 -69.2437>}]
 dyn_anemone3_05 [{shapeType=Capsule, length=1, radius=1, rotation=<rot 0 -90 0>}]
 dyn_anemone3_06 [{shapeType=Capsule, length=1, radius=1, center=<3.413 0.0646 0.8553>, rotation=<rot 0 85.3633 -57.5021>}]
 dyn_anemone3_07 [{shapeType=Capsule, length=1, radius=1, center=<0.2304 -2.195 1.6601>, rotation=<rot 0 -162.8624 -49.9723>}]
 dyn_anemone3_08 [{shapeType=Capsule, length=1, radius=1, center=<1.8118 0.1167 0.4492>, rotation=<rot 0 84.6461 -173.1221>}]
 dyn_anemone3_09 [{shapeType=Capsule, length=1, radius=1, center=<0.5484 -0.3812 0.3947>, rotation=<rot -2.6391 -90 1.7252>}]
 dyn_anemone3_10 [{shapeType=Capsule, length=1, radius=1, center=<1.5768 -0.1908 0.1615>, rotation=<rot -27.6255 -90 5.5421>}]
 dyn_anemone3_11 [{shapeType=Capsule, length=1, radius=1, rotation=<rot 0 -90 0>}]
 dyn_anemone3_12 [{shapeType=Capsule, length=1, radius=1, center=<2.5468 -0.6897 0.7088>, rotation=<rot 0 60.8552 -53.1624>}]
 dyn_anemone3_13 [{shapeType=Capsule, length=1, radius=1, center=<1.4438 -1.3395 2.4692>, rotation=<rot 0 76.6856 -87.0965>}]
 dyn_anemone4_01 [{shapeType=Capsule, length=0.5, radius=0.5}]
 dyn_anemone4_02 [{shapeType=Capsule, length=0.505, radius=0.505}]
 dyn_anemone5_01 [{shapeType=Capsule, length=0.5, radius=0.5}]
 dyn_anemone5_02 [{shapeType=Capsule, length=0.505, radius=0.505}]
 pack [{shapeType=Capsule, length=1, radius=1}]
}
constraints [
 {bone1Name=dyn_anemone1_01, bone2Name=pack, swing1LimitDegrees=5, swing2LimitDegrees=5, twistLimitDegrees=5}
 {bone1Name=dyn_anemone1_02, bone2Name=dyn_anemone1_01, swing1LimitDegrees=5, swing2LimitDegrees=5, twistLimitDegrees=5}
 {bone1Name=dyn_anemone1_03, bone2Name=dyn_anemone1_02, swing1LimitDegrees=5, swing2LimitDegrees=5, twistLimitDegrees=5}
 {bone1Name=dyn_anemone1_04, bone2Name=dyn_anemone1_03, swing1LimitDegrees=5, swing2LimitDegrees=5, twistLimitDegrees=5}
 {bone1Name=dyn_anemone1_05, bone2Name=dyn_anemone1_04, swing1LimitDegrees=5, swing2LimitDegrees=5, twistLimitDegrees=5}
 {bone1Name=dyn_anemone1_06, bone2Name=dyn_anemone1_02, swing1LimitDegrees=5, swing2LimitDegrees=5, twistLimitDegrees=5}
 {bone1Name=dyn_anemone1_07, bone2Name=dyn_anemone1_06, swing1LimitDegrees=5, swing2LimitDegrees=5, twistLimitDegrees=5}
 {bone1Name=dyn_anemone1_08, bone2Name=dyn_anemone1_07, swing1LimitDegrees=5, swing2LimitDegrees=5, twistLimitDegrees=5}
 {bone1Name=dyn_anemone1_09, bone2Name=dyn_anemone1_02, swing1LimitDegrees=5, swing2LimitDegrees=5, twistLimitDegrees=5}
 {bone1Name=dyn_anemone1_10, bone2Name=dyn_anemone1_09, swing1LimitDegrees=5, swing2LimitDegrees=5, twistLimitDegrees=5}
 {bone1Name=dyn_anemone1_11, bone2Name=dyn_anemone1_10, swing1LimitDegrees=5, swing2LimitDegrees=5, twistLimitDegrees=5}
 {bone1Name=dyn_anemone1_12, bone2Name=dyn_anemone1_02, swing1LimitDegrees=5, swing2LimitDegrees=5, twistLimitDegrees=5}
 {bone1Name=dyn_anemone1_13, bone2Name=dyn_anemone1_12, swing1LimitDegrees=5, swing2LimitDegrees=5, twistLimitDegrees=5}
 {bone1Name=dyn_anemone1_14, bone2Name=dyn_anemone1_13, swing1LimitDegrees=5, swing2LimitDegrees=5, twistLimitDegrees=5}
 {bone1Name=dyn_anemone1_15, bone2Name=dyn_anemone1_02, swing1LimitDegrees=5, swing2LimitDegrees=5, twistLimitDegrees=5}
 {bone1Name=dyn_anemone1_16, bone2Name=dyn_anemone1_15, swing1LimitDegrees=5, swing2LimitDegrees=5, twistLimitDegrees=5}
 {bone1Name=dyn_anemone1_17, bone2Name=dyn_anemone1_16, swing1LimitDegrees=5, swing2LimitDegrees=5, twistLimitDegrees=5}
 {bone1Name=dyn_anemone1_18, bone2Name=dyn_anemone1_02, swing1LimitDegrees=5, swing2LimitDegrees=5, twistLimitDegrees=5}
 {bone1Name=dyn_anemone1_19, bone2Name=dyn_anemone1_18, swing1LimitDegrees=5, swing2LimitDegrees=5, twistLimitDegrees=5}
 {bone1Name=dyn_anemone2_01, bone2Name=pack, swing1LimitDegrees=5, swing2LimitDegrees=5, twistLimitDegrees=5}
 {bone1Name=dyn_anemone2_02, bone2Name=dyn_anemone2_01, swing1LimitDegrees=5, swing2LimitDegrees=5, twistLimitDegrees=5}
 {bone1Name=dyn_anemone2_03, bone2Name=dyn_anemone2_02, swing1LimitDegrees=5, swing2LimitDegrees=5, twistLimitDegrees=5}
 {bone1Name=dyn_anemone2_04, bone2Name=dyn_anemone2_01, swing1LimitDegrees=5, swing2LimitDegrees=5, twistLimitDegrees=5}
 {bone1Name=dyn_anemone2_05, bone2Name=dyn_anemone2_04, swing1LimitDegrees=5, swing2LimitDegrees=5, twistLimitDegrees=5}
 {bone1Name=dyn_anemone2_06, bone2Name=dyn_anemone2_05, swing1LimitDegrees=5, swing2LimitDegrees=5, twistLimitDegrees=5}
 {bone1Name=dyn_anemone2_07, bone2Name=dyn_anemone2_01, swing1LimitDegrees=5, swing2LimitDegrees=5, twistLimitDegrees=5}
 {bone1Name=dyn_anemone2_08, bone2Name=dyn_anemone2_07, swing1LimitDegrees=5, swing2LimitDegrees=5, twistLimitDegrees=5}
 {bone1Name=dyn_anemone2_09, bone2Name=dyn_anemone2_01, swing1LimitDegrees=5, swing2LimitDegrees=5, twistLimitDegrees=5}
 {bone1Name=dyn_anemone2_10, bone2Name=dyn_anemone2_09, swing1LimitDegrees=5, swing2LimitDegrees=5, twistLimitDegrees=5}
 {bone1Name=dyn_anemone3_01, bone2Name=pack, swing1LimitDegrees=5, swing2LimitDegrees=5, twistLimitDegrees=5}
 {bone1Name=dyn_anemone3_04, bone2Name=dyn_anemone3_01, swing1LimitDegrees=5, swing2LimitDegrees=5, twistLimitDegrees=5}
 {bone1Name=dyn_anemone3_05, bone2Name=dyn_anemone3_04, swing1LimitDegrees=5, swing2LimitDegrees=5, twistLimitDegrees=5}
 {bone1Name=dyn_anemone3_06, bone2Name=dyn_anemone3_01, swing1LimitDegrees=5, swing2LimitDegrees=5, twistLimitDegrees=5}
 {bone1Name=dyn_anemone3_07, bone2Name=dyn_anemone3_06, swing1LimitDegrees=5, swing2LimitDegrees=5, twistLimitDegrees=5}
 {bone1Name=dyn_anemone3_08, bone2Name=dyn_anemone3_01, swing1LimitDegrees=5, swing2LimitDegrees=5, twistLimitDegrees=5}
 {bone1Name=dyn_anemone3_09, bone2Name=dyn_anemone3_08, swing1LimitDegrees=5, swing2LimitDegrees=5, twistLimitDegrees=5}
 {bone1Name=dyn_anemone3_10, bone2Name=dyn_anemone3_01, swing1LimitDegrees=5, swing2LimitDegrees=5, twistLimitDegrees=5}
 {bone1Name=dyn_anemone3_11, bone2Name=dyn_anemone3_10, swing1LimitDegrees=5, swing2LimitDegrees=5, twistLimitDegrees=5}
 {bone1Name=dyn_anemone3_12, bone2Name=dyn_anemone3_01, swing1LimitDegrees=5, swing2LimitDegrees=5, twistLimitDegrees=5}
 {bone1Name=dyn_anemone3_13, bone2Name=dyn_anemone3_12, swing1LimitDegrees=5, swing2LimitDegrees=5, twistLimitDegrees=5}
 {bone1Name=dyn_anemone4_01, bone2Name=pack, swing1LimitDegrees=5, swing2LimitDegrees=5, twistLimitDegrees=5}
 {bone1Name=dyn_anemone4_02, bone2Name=dyn_anemone4_01, swing1LimitDegrees=5, swing2LimitDegrees=5, twistLimitDegrees=5}
 {bone1Name=dyn_anemone5_01, bone2Name=pack, swing1LimitDegrees=5, swing2LimitDegrees=5, twistLimitDegrees=5}
 {bone1Name=dyn_anemone5_02, bone2Name=dyn_anemone5_01, swing1LimitDegrees=5, swing2LimitDegrees=5, twistLimitDegrees=5}
]
