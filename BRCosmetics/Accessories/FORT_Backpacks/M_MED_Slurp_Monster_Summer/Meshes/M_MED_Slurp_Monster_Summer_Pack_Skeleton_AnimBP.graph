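# Animation Blueprint for the summer Slurp Monster backpack cosmetic, playing an idle and a hit reaction when the player takes damage.
# Health or shield decrease input events briefly set Took Damage, triggering the reaction, which returns automatically to idle.
# The state machine feeds RigidBody physics simulated relative to the pack bone, then an additive rotation on one anemone bone.
# An older-style health-comparison chain and a Copy Pose node sit unconnected.
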AnimBlueprint /BRCosmetics/Accessories/FORT_Backpacks/M_MED_Slurp_Monster_Summer/Meshes/M_MED_Slurp_Monster_Summer_Pack_Skeleton_AnimBP
target_skeleton=&/M_MED_Slurp_Monster_Summer_Pack_Skeleton
graphs [
 AnimGraph {nodes [
   Copy Pose From Mesh()
   AnimGraphNode_StateMachine_0 = State Machine() {is=title AnimGraphNode_StateMachine}
   n1 = RigidBody(ComponentPose=Local To Component(LocalPose=@AnimGraphNode_StateMachine_0)) {
    bClampLinearTranslationLimitToRefPose=true
    bOverrideWorldGravity=true
    overridePhysicsAsset=&/M_MED_Slurp_Monster_Summer_Pack_PhysicsAsset
    simulationSpace=BaseBoneSpace
    alphaBoolBlend {blendOption=Linear}
    baseBoneRef {boneName=pack}
    componentAppliedLinearAccClamp=<2000 2000 2000>
    componentLinearAccScale=<0.15 0.3 0.3>
    componentLinearVelScale=<0.3 0.3 0.3>
    overrideWorldGravity {z=-980}
   }
   n2 = "Transform (Modify) Bone - Bone: dyn_anemone1_01"(ComponentPose=@n1, Alpha=1, Translation="0,0,0", Rotation="-10,0,0", Scale="1,1,1") {
    rotationMode=BMM_Additive
    rotationSpace=BCS_BoneSpace
    alphaBoolBlend {blendOption=Linear}
    boneToModify {boneName=dyn_anemone1_01}
    rotation {pitch=-10}
   }
   Output Pose(Result=Component To Local(ComponentPose=@n2)) {layerGroup=None}
  ]}
 State Machine {
  entry=Idle
  path=AnimGraph.AnimGraphNode_StateMachine_0.State Machine
  ids=id
  nodes [
   AnimStateNode_0 = Idle()
   AnimStateNode_1 = Hit React() {is=title AnimStateNode}
   AnimStateTransitionNode_0 = AnimStateTransitionNode()
   AnimStateTransitionNode_1 = AnimStateTransitionNode() {bAutomaticRuleBasedOnSequencePlayerInState=true}
  ]
  links [
   {
    dst=Idle
    rule=AnimGraph.AnimGraphNode_StateMachine_0.State Machine.AnimStateTransitionNode_1.Transition
    src=Hit React
    via=AnimStateTransitionNode_1
   }
   {
    dst=Hit React
    rule=AnimGraph.AnimGraphNode_StateMachine_0.State Machine.AnimStateTransitionNode_0.Transition
    src=Idle
    via=AnimStateTransitionNode_0
   }
  ]
 }
 Idle {
  path=AnimGraph.AnimGraphNode_StateMachine_0.State Machine.AnimStateNode_0.Idle
  nodes [Output Animation Pose(Result=Sequence Player 'Slurp_Monster_Summer_Idle_BACKPACK'() {sequence=/BRCosmetics/Accessories/FORT_Backpacks/M_MED_Slurp_Monster_Summer/Animation/Slurp_Monster_Summer_Idle_BACKPACK})]
 }
 Hit React {
  path=AnimGraph.AnimGraphNode_StateMachine_0.State Machine.AnimStateNode_1.Hit React
  nodes [
   n1 = Sequence Player 'Slurp_Monster_Summer_HitReact_BACKPACK'() {
    bLoopAnimation=false
    sequence=/BRCosmetics/Accessories/FORT_Backpacks/M_MED_Slurp_Monster_Summer/Animation/Slurp_Monster_Summer_HitReact_BACKPACK
   }
   Output Animation Pose(Result=@n1)
  ]
 }
 Transition {
  path=AnimGraph.AnimGraphNode_StateMachine_0.State Machine.AnimStateTransitionNode_0.Transition
  nodes [Result(bCanEnterTransition=Get Took Damage() -> (Took Damage:bool))]
 }
 Transition {path=AnimGraph.AnimGraphNode_StateMachine_0.State Machine.AnimStateTransitionNode_1.Transition, nodes [Result()]}
 EventGraph {nodes [
   n1 = Cast To FortPlayerPawn(Object=GetOwningActor())
   n2 = GetHealth!(self:Fort Pawn Object Reference=@n1) -> (ReturnValue:float)
   > Set Current Health(Current Health:double=@n2) -> (Output_Get:double)
   > n3 = Branch(Condition="Not Equal (Float)"!(A:double=@n2, B:double=Get Previous Health() -> (Previous Health:double)) -> (ReturnValue:bool) {is=title K2Node_CallFunction})
   > Set Took Damage(Took Damage:bool=true) -> (Output_Get:bool)
   > Set Previous Health(Previous Health:double=Get Current Health() -> (Current Health:double)) -> (Output_Get:double)
   Set Took Damage(execute=@n3.else, Took Damage=false)
   n4 = Event BlueprintProcessAnimInputEvent()
   > Branch(Condition=OR Boolean(
     A=IsNamed(self=@n4.AnimInputEvent, InEventName=aie.PlayerVitality.HealthDecreased)
     B=IsNamed(self=@n4.AnimInputEvent, InEventName=aie.PlayerVitality.ShieldsDecreased)
    ))
   > Set Took Damage(Took Damage=true)
   > RetriggerableDelay(Duration=0.2)
   > Set Took Damage(Took Damage=false)
   TryGetPawnOwner()
   Event BlueprintUpdateAnimation()
  ]}
]
variables [
 Current Health {type=number}
 Previous Health {type=number, default=100}
 Took Damage {type=boolean}
]
